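# Post-copy AnimBP for the Future Samurai Summer Alt face accessory, adding head-driven corrective posing and secondary motion to the rope and hair.
# Each update derives up and down alphas from head rotation relative to neck_02.
# After copying the attached parent's pose, these alphas drive a neck_01 rotation constraint and additive head and rope offsets.
# Spring-based Anim Dynamics then simulates the head rope and five hair bones with tight angular limits.
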AnimBlueprint /BRCosmetics/Characters/Player/Male/Medium/Bodies/M_MED_Future_Samurai_Summer_Alt/Meshes/Parts/M_MED_Future_Samurai_Summer_Alt_FaceAcc_AnimBP
target_skeleton=&/M_MED_Future_Samurai_Summer_Alt_FaceAcc_Skeleton
graphs [
 AnimGraph {nodes [
   n1 = Get neck_rot_up() -> (neck_rot_up:double)
   n2 = "Constraint - neck_01 (RR)"!(
    ComponentPose=Local To Component(LocalPose=Copy Pose From Mesh() {bUseAttachedParent=true})
    Alpha=@n1
    ConstraintWeights_0=0.7
    ConstraintWeights_1:float=0.3
   ) -> (Pose) {
    alphaBoolBlend {blendOption=Linear}
    boneToModify {boneName=neck_01}
    constraintSetup [{transformType=Rotation, targetBone {boneName=spine_05}}, {transformType=Rotation, targetBone {boneName=head}}]
    constraintWeights [1,1]
   }
   n3 = "Transform (Modify) Bone - Bone: head"(ComponentPose=@n2, Alpha=@n1, Translation="0,0,0", Rotation="0,20,0", Scale="1,1,1") {
    rotationMode=BMM_Additive
    rotationSpace=BCS_BoneSpace
    scaleMode=BMM_Additive
    scaleSpace=BCS_BoneSpace
    translationMode=BMM_Additive
    translationSpace=BCS_BoneSpace
    alphaBoolBlend {blendOption=Linear}
    boneToModify {boneName=head}
    rotation {yaw=20}
   }
   n4 = "Transform (Modify) Bone - Bone: dyn_head_rope"(ComponentPose=@n3, Alpha=@n1, Translation="3,3,15", Rotation="-60,0,-40", Scale="1,1,1") {
    rotationMode=BMM_Additive
    rotationSpace=BCS_BoneSpace
    scaleMode=BMM_Additive
    scaleSpace=BCS_BoneSpace
    translationMode=BMM_Additive
    translationSpace=BCS_BoneSpace
    alphaBoolBlend {blendOption=Linear}
    boneToModify {boneName=dyn_head_rope}
    rotation {pitch=-60, roll=-40}
    translation=<3 3 15>
   }
   n5 = "Transform (Modify) Bone - Bone: head"(ComponentPose=@n4, Alpha=Get neck_rot_dn() -> (neck_rot_dn:double), Translation="0,0,0", Rotation="0,-30,0", Scale="1,1,1") {
    rotationMode=BMM_Additive
    rotationSpace=BCS_BoneSpace
    scaleMode=BMM_Additive
    scaleSpace=BCS_BoneSpace
    translationMode=BMM_Additive
    translationSpace=BCS_BoneSpace
    alphaBoolBlend {blendOption=Linear}
    boneToModify {boneName=head}
    rotation {yaw=-30}
   }
   n6 = Anim Dynamics - Bone: dyn_head_rope(ComponentPose=@n5, Alpha=1, ExternalForce="0,0,0") {
    angularSpringConstant=800
    bAngularSpring=true
    bGravityOverrideInSimSpace=true
    bOverrideAngularDamping=true
    bUseGravityOverride=true
    gravityScale=0.7
    simulationSpace=RootRelative
    alphaBoolBlend {blendOption=Linear}
    boundBone {boneName=dyn_head_rope}
    componentAppliedLinearAccClamp=<300 300 300>
    componentLinearAccScale=<0.1 0.1 0.15>
    componentLinearVelScale=<0.1 0.1 0.15>
    gravityOverride {z=-980}
    physicsBodyDefinitions [{
      sphereCollisionRadius=0
      boundBone {boneName=dyn_head_rope}
      boxExtents=<10 15 15>
      constraintSetup {angularLimitsMax=<10 5 7>, angularLimitsMin=<-10 -5 -7>, angularTarget=<0 0 0>}
      localJointOffset=<-5 -0.8 -0.5>
     }]
    relativeSpaceBone {boneName=pelvis}
   }
   n7 = Anim Dynamics - Bone: dyn_hair_01(ComponentPose=@n6, Alpha=1) {
    angularSpringConstant=800
    bAngularSpring=true
    bGravityOverrideInSimSpace=true
    bOverrideAngularDamping=true
    bUseGravityOverride=true
    gravityScale=0.7
    simulationSpace=BoneRelative
    alphaBoolBlend {blendOption=Linear}
    boundBone {boneName=dyn_hair_01}
    componentAppliedLinearAccClamp=<300 300 300>
    componentLinearAccScale=<0.1 0.1 0.15>
    componentLinearVelScale=<0.1 0.1 0.15>
    gravityOverride {z=-980}
    physicsBodyDefinitions [{
      sphereCollisionRadius=0
      boundBone {boneName=dyn_hair_01}
      boxExtents=<10 15 15>
      constraintSetup {twistAxis=AxisY, angularLimitsMax=<10 5 7>, angularLimitsMin=<-10 -5 -7>, angularTarget=<0 0 0>}
      localJointOffset=<-5 0.8 -0.5>
     }]
    relativeSpaceBone {boneName=root}
   }
   n8 = Anim Dynamics - Bone: dyn_hair_02(ComponentPose=@n7, Alpha=1) {
    angularSpringConstant=800
    bAngularSpring=true
    bGravityOverrideInSimSpace=true
    bOverrideAngularDamping=true
    bUseGravityOverride=true
    gravityScale=0.7
    simulationSpace=BoneRelative
    alphaBoolBlend {blendOption=Linear}
    boundBone {boneName=dyn_hair_02}
    componentAppliedLinearAccClamp=<300 300 300>
    componentLinearAccScale=<0.1 0.1 0.15>
    componentLinearVelScale=<0.1 0.1 0.15>
    gravityOverride {z=-980}
    physicsBodyDefinitions [{
      sphereCollisionRadius=0
      boundBone {boneName=dyn_hair_02}
      boxExtents=<10 15 15>
      constraintSetup {twistAxis=AxisY, angularLimitsMax=<10 5 7>, angularLimitsMin=<-10 -5 -7>, angularTarget=<0 0 0>}
      localJointOffset=<-5 0.8 -0.5>
     }]
    relativeSpaceBone {boneName=root}
   }
   n9 = Anim Dynamics - Bone: dyn_hair_03(ComponentPose=@n8, Alpha=1, ExternalForce="0,0,0") {
    angularSpringConstant=800
    bAngularSpring=true
    bGravityOverrideInSimSpace=true
    bOverrideAngularDamping=true
    bUseGravityOverride=true
    gravityScale=0.7
    simulationSpace=BoneRelative
    alphaBoolBlend {blendOption=Linear}
    boundBone {boneName=dyn_hair_03}
    componentAppliedLinearAccClamp=<300 300 300>
    componentLinearAccScale=<0.1 0.1 0.15>
    componentLinearVelScale=<0.1 0.1 0.15>
    gravityOverride {z=-980}
    physicsBodyDefinitions [{
      sphereCollisionRadius=0
      boundBone {boneName=dyn_hair_03}
      boxExtents=<10 15 15>
      constraintSetup {twistAxis=AxisY, angularLimitsMax=<10 5 7>, angularLimitsMin=<-10 -5 -7>, angularTarget=<0 0 0>}
      localJointOffset=<-5 0.8 -0.5>
     }]
    relativeSpaceBone {boneName=root}
   }
   n10 = Anim Dynamics - Bone: dyn_hair_04(ComponentPose=@n9, Alpha=1, ExternalForce="0,0,0") {
    angularSpringConstant=800
    bAngularSpring=true
    bGravityOverrideInSimSpace=true
    bOverrideAngularDamping=true
    bUseGravityOverride=true
    gravityScale=0.7
    simulationSpace=BoneRelative
    alphaBoolBlend {blendOption=Linear}
    boundBone {boneName=dyn_hair_04}
    componentAppliedLinearAccClamp=<300 300 300>
    componentLinearAccScale=<0.1 0.1 0.15>
    componentLinearVelScale=<0.1 0.1 0.15>
    gravityOverride {z=-980}
    physicsBodyDefinitions [{
      sphereCollisionRadius=0
      boundBone {boneName=dyn_hair_04}
      boxExtents=<10 15 15>
      constraintSetup {twistAxis=AxisY, angularLimitsMax=<10 5 10>, angularLimitsMin=<-10 -5 -10>, angularTarget=<0 0 0>}
      localJointOffset=<-5 0.8 -0.5>
     }]
    relativeSpaceBone {boneName=root}
   }
   n11 = Anim Dynamics - Bone: dyn_hair_05(ComponentPose=@n10, Alpha=1, ExternalForce="0,0,0") {
    angularSpringConstant=800
    bAngularSpring=true
    bGravityOverrideInSimSpace=true
    bOverrideAngularDamping=true
    bUseGravityOverride=true
    gravityScale=0.7
    simulationSpace=BoneRelative
    alphaBoolBlend {blendOption=Linear}
    boundBone {boneName=dyn_hair_05}
    componentAppliedLinearAccClamp=<300 300 300>
    componentLinearAccScale=<0.1 0.1 0.15>
    componentLinearVelScale=<0.1 0.1 0.15>
    gravityOverride {z=-980}
    physicsBodyDefinitions [{
      sphereCollisionRadius=0
      boundBone {boneName=dyn_hair_05}
      boxExtents=<10 15 15>
      constraintSetup {twistAxis=AxisY, angularLimitsMax=<10 5 10>, angularLimitsMin=<-10 -5 -10>, angularTarget=<0 0 0>}
      localJointOffset=<-5 -0.8 -0.5>
     }]
    relativeSpaceBone {boneName=root}
   }
   Output Pose(Result=Component To Local(ComponentPose=@n11)) {layerGroup=None}
  ]}
 EventGraph {nodes [
   Event BlueprintUpdateAnimation()
   > Set neck_rot_up(neck_rot_up:double=GetAlphaFromReferenceTransformDelta(SkeletalMeshComponent=Get Skel_Mesh(), BoneName=head, BaseName=neck_02, InputRange=(X=0,Y=-40), TwistAxis="0,0,1")) -> (Output_Get:double)
   > Set neck_rot_dn(neck_rot_dn:double=GetAlphaFromReferenceTransformDelta(SkeletalMeshComponent=Get Skel_Mesh(), BoneName=head, BaseName=neck_02, InputRange=(X=-15,Y=30), TwistAxis="0,0,1")) -> (Output_Get:double)
   Event BlueprintInitializeAnimation()
   > Set Skel_Mesh(Skel_Mesh=GetOwnerSkeletalMeshComponent())
  ]}
]
variables [
 neck_rot_dn {type=number}
 neck_rot_up {type=number}
 Skel_Mesh {type=SkeletalMeshComponent}
]
